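# Niagara system for the CrimpPerfume player emote that shows leafy plant meshes.
# A single emitter, LeafMeshesCondensed, burst-spawns mesh particles with randomized green tints that persist until the emitter deactivates rather than expiring by lifetime.
# A curve on emitter age quickly scales the meshes up from zero, while a slow waveform lerps between two nearly identical facing vectors for a subtle sway.
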NiagaraSystem /BRCosmetics/Animation/Game/MainPlayer/Emotes/CrimpPerfume/VFX/NS_Emote_CrimpPerfume_Plants
dependencies {usedDataInterfaces [&NiagaraDataInterfaceCurve], usedDynamicInputs [], usedModules [&SystemState]}
emitters {LeafMeshesCondensed {
  emitterUpdateScript [
   &EmitterState(Life Cycle Mode=::NewEnumerator1, Loop Behavior=::NewEnumerator1, Loop Duration Mode=::NewEnumerator1, UseLoopDelay=false)
   &SpawnBurst_Instantaneous(Use Spawn Probability=false, UseLoopCountLimit=false)
  ]
  particleSpawnScript [&InitializeParticle(
    Color Maximum=<rgba 0.1422 0.2865 0.0863 1>
    Color Minimum=<rgba 0.2392 0.3646 0.1487 1>
    Color Mode=::NewEnumerator2
    Lifetime=$Emitter.CurrentLoopDuration
    Mesh Scale Mode=::NewEnumerator3
    UsePositionOffset=false
    Write Material Random=true
   )]
  particleUpdateScript [
   &ParticleState(Kill Particles When Lifetime Has Elapsed=false, Let Infinitely Lived Particles Die When Emitter Deactivates=true)
   &UpdateMeshOrientation(
    Facing Direction=&Lerp_Vector(
     A=<1 0 0>
     B=<1 0.01 0.02>
     Alpha=&Waveform("Waveform [1]"=::NewEnumerator0, "[1] Frequency"=0.3, "[1] Amplitude Min/Max"=<0 1>)
    )
    Facing Interpolation=false
    Orientation Method=::NewEnumerator1
    Override Quat=false
    TransformFacing=false
    TransformUp=false
   )
   &ScaleMeshSize(Scale Factor=&VectorFromFloat(Value=&FloatFromCurve(FloatCurve=DataInterface() {propertyValues {Curve {keys [<at 0 0>,<at 0.1 1>]}}}, CurveIndex=$Emitter.Age)))
   &SolveForcesAndVelocity(Clamp Velocity=false, Limit Acceleration=false)
  ]
  rendererClasses [&NiagaraMeshRendererProperties]
  renderers [{rendererClass=&NiagaraMeshRendererProperties, rendererIndex=0}]
 }}
summary {
 systemName=NS_Emote_CrimpPerfume_Plants
 emitters [{emitterName=LeafMeshesCondensed, rendererClasses [&NiagaraMeshRendererProperties]}]
 userVariables []
}
