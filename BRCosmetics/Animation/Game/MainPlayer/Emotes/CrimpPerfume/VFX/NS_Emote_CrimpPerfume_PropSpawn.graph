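# Niagara system for the CrimpPerfume emote's prop spawn, bursting rocks, smoke and leaves around a chair and umbrella.
# Umbrella rocks fire shortly before the delayed chair rocks, smoke and leaves; UmbrellaSmoke has no renderers.
# Rock emitters split three spawn groups across mesh and sprite renderers via VisibilityTag, with gravity and two collision modules gated by quality level or front-end preview.
# Front-end flags also switch the chair smoke color.
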NiagaraSystem /BRCosmetics/Animation/Game/MainPlayer/Emotes/CrimpPerfume/VFX/NS_Emote_CrimpPerfume_PropSpawn
dependencies {
 usedDataInterfaces [&NiagaraDataInterfacePlatformSet, &NiagaraDataInterfaceVectorCurve, &NiagaraDataInterfaceCurve]
 usedDynamicInputs []
 usedModules [
  &SystemState
  &self:SystemSpawnScript.SystemScriptSource.SystemScriptGraph.NiagaraNodeAssignment_0.SetVariables_#1
 ]
}
emitters {
 ChairRocks {
  emitterUpdateScript [
   &EmitterState(Life Cycle Mode=::NewEnumerator1, Loop Behavior=::NewEnumerator1, Loop Delay=0.23, Loop Duration=4, UseLoopDelay=true)
   &SpawnBurst_Instantaneous(
    Spawn Count=&RandomRangeInt(Minimum=12, Maximum=20, Recalculate Random Each Loop=false)
    Use Spawn Probability=false
    UseLoopCountLimit=false
   )
   &SpawnBurst_Instantaneous(Spawn Count=6, Spawn Group=1, Use Spawn Probability=false, UseLoopCountLimit=false) {module=SpawnBurst_Instantaneous001}
   &SpawnBurst_Instantaneous(Spawn Count=40, Spawn Group=2, Use Spawn Probability=false, UseLoopCountLimit=false) {module=SpawnBurst_Instantaneous002}
  ]
  particleSpawnScript [
   /SetVariables(Particles.VisibilityTag=&SpawnGroup()) {isSetParameters=true, module=SetVariables_#2}
   &InitializeParticle(
    Color=<rgba 0.1406 0.0699 0.0342 1>
    Lifetime=&MergeHelper_FloatSwitch(
     0=&RandomRangeFloat(Minimum=0.9, Maximum=1)
     1=&RandomRangeFloat(Minimum=0.9, Maximum=1)
     2=&RandomRangeFloat(Minimum=0.6, Maximum=1)
     3=?
     Index=&SpawnGroup()
     Num=::NewEnumerator1
    )
    Mass Max=1.5
    Mass Mode=::NewEnumerator3
    Mesh Scale Mode=::NewEnumerator3
    Mesh Uniform Scale=&MaskFloatBySpawnGroup(
     Masked Float=&RandomRangeFloat(Minimum=0.05, Maximum=0.1)
     Passthrough Float=&RandomRangeFloat(Minimum=0.15, Maximum=0.19)
    )
    Sprite Rotation Mode=::NewEnumerator2
    Sprite Size Mode=::NewEnumerator3
    Uniform Sprite Size=&RandomRangeFloat(Minimum=3, Maximum=5)
    UsePositionOffset=true
    Write Material Random=true
   )
   &ShapeLocation(
    MaskSpawnGroup=false
    Non Uniform Scale=<1 3 1>
    Offset=<0 0 12>
    Offset Mode=::NewEnumerator0
    Override Seed=false
    Ring Radius=23
    Shape Primitive=::NewEnumerator4
    UseOwnerScale=true
   )
   &InitialMeshOrientation(Use Rotation Vector=false)
   &AddVelocity(
    Cone Angle=&MaskFloatBySpawnGroup(
     Masked Float=&RandomRangeFloat(Minimum=40.8464, Maximum=186.7016)
     Passthrough Float=&RandomRangeFloat(Minimum=14.1396, Maximum=44.0656)
    )
    Module.UseConeFalloff=false
    Module.UseVelDistribution=false
    Override Seed=false
    Rotation Angle=180
    Rotation Mode=::NewEnumerator1
    Velocity Mode=::NewEnumerator2
    Velocity Speed=&MergeHelper_FloatSwitch(
     0=&RandomRangeFloat(Minimum=50, Maximum=350)
     1=&RandomRangeFloat(Minimum=20, Maximum=200)
     2=&RandomRangeFloat(Minimum=100, Maximum=400)
     3=?
     Index=&SpawnGroup()
     Num=::NewEnumerator1
    )
   )
  ]
  particleUpdateScript [
   &GravityForce(Gravity=<0 0 -1180.0895>)
   &Collision(
    Age Colliding Particles=false
    Collision Enabled=&Boolean_OrOperation(
     Bool1=/Niagara/DynamicInputs/Helpers/MaskBool(
      Mask Source:NiagaraInt32=0
      Mask Parameter:NiagaraInt32=?
      Mask:NiagaraInt32=2
      Masked Bool:NiagaraBool=false
      Passthrough Bool:NiagaraBool=&Boolean_OrOperation(
       Bool1=&IsPlatformSetActive(New Platform Set=DataInterface() {propertyValues {Platforms {qualityLevelMask=-8}}})
       Bool2=$User.bIsFrontEndPreview
      )
     )
     Bool2=false
    )
    Friction=0.5
    Friction During a Bounce=0.3
    KillOnCollision=false
    Module.Control Rest State Via Surface Normal=false
    Module.EnableMaxCollisionCount=false
    Module.Randomize Collision Normal Vector=false
    Radius Calculation Type=::NewEnumerator1
    Restitution=0.44
   )
   &Collision(
    Age Colliding Particles=false
    Analytical Collision Plane Position 1:Vector3f=<0 0 0>
    Analytical Collision Plane Position 2:Vector3f=<0 0 0>
    Analytical Collision Plane Space:ENiagaraCoordinateSpace=::Local
    CPU Collision Type=::NewEnumerator3
    Collision Enabled=&Boolean_OrOperation(
     Bool1=&MaskBool(
      Mask Source:NiagaraInt32=0
      Mask Parameter:NiagaraInt32=?
      Mask:NiagaraInt32=2
      Masked Bool:NiagaraBool=false
      Passthrough Bool:NiagaraBool=&Boolean_OrOperation(
       Bool1=&IsPlatformSetActive(New Platform Set=DataInterface() {propertyValues {Platforms {qualityLevelMask=-28}}})
       Bool2=$User.bIsFrontEndPreview
      )
     )
     Bool2=false
    )
    Friction=0.5
    Friction During a Bounce=0.3
    KillOnCollision=false
    Module.Control Rest State Via Surface Normal=false
    Module.EnableMaxCollisionCount=false
    Module.Randomize Collision Normal Vector=false
    Radius Calculation Type=::NewEnumerator1
    Restitution=0.44
   ) {module=Collision001}
   &SolveForcesAndVelocity(Clamp Velocity=false, Limit Acceleration=false)
   &ParticleState()
   &ScaleMeshSize(Scale Factor=&VectorFromCurve(VectorCurve=DataInterface() {propertyValues {
       XCurve {keys [<at 0 0.35 RCIM_Cubic>, {arriveTangent=-0.35, interpMode=RCIM_Cubic, leaveTangent=-0.35, time=0.1, value=1}, <at 1 0 RCIM_Cubic>]}
       YCurve {keys [<at 0 0.35 RCIM_Cubic>, {arriveTangent=-0.35, interpMode=RCIM_Cubic, leaveTangent=-0.35, time=0.1, value=1}, <at 1 0 RCIM_Cubic>]}
       ZCurve {keys [<at 0 0.35 RCIM_Cubic>, {arriveTangent=-0.35, interpMode=RCIM_Cubic, leaveTangent=-0.35, time=0.1, value=1}, <at 1 0 RCIM_Cubic>]}
      }}))
   &DynamicMaterialParameters(Module.Index0Float0Write=true, Module.Index0Float1Write=true, Module.Index0Float2Write=true, Module.Index0Float3Write=true)
   /SetVariables(Particles.float:NiagaraFloat=&FloatFromCurve(
     FloatCurve=DataInterface() {propertyValues {Curve {
        defaultValue=1
        keys [
         {arriveTangent=4.1785, interpMode=RCIM_Cubic, leaveTangent=4.1785, tangentMode=RCTM_User, time=0, value=0}
         <at 0.2318 0.8879 RCIM_Cubic>
         {arriveTangent=0.0224, interpMode=RCIM_Cubic, leaveTangent=0.0224, tangentMode=RCTM_User, time=0.9929, value=0}
        ]
       }}}
     CurveIndex=$Particles.Age
    )) {isSetParameters=true, module=SetVariables_#3}
  ]
  rendererClasses [&NiagaraMeshRendererProperties, &NiagaraSpriteRendererProperties]
  renderers [
   {rendererClass=&NiagaraMeshRendererProperties, rendererIndex=0}
   {rendererClass=&NiagaraSpriteRendererProperties, rendererIndex=1}
   {rendererClass=&NiagaraMeshRendererProperties, rendererIndex=2}
   {rendererClass=&NiagaraSpriteRendererProperties, rendererIndex=3}
  ]
 }
 Chair_Smoke {
  emitterUpdateScript [
   &EmitterState(Life Cycle Mode=::NewEnumerator1, Loop Behavior=::NewEnumerator1, Loop Delay=0.23, Loop Duration=3, UseLoopDelay=true)
   &SpawnBurst_Instantaneous(Spawn Count=45, Use Spawn Probability=false, UseLoopCountLimit=false)
  ]
  particleSpawnScript [
   &InitializeParticle(
    Color=&MakeCustomLinearColorFromBool(
     Bool=&Boolean_OrOperation(Bool1=$User.bIsFrontEndPreview, Bool2=$User.bIsFrontEnd)
     True Color=<rgba 0.5469 0.343 0.2649 1>
     False Color=<rgba 0.3542 0.2341 0.1882 1>
    )
    Lifetime Max=1.3
    Lifetime Min=0.33
    Lifetime Mode=::NewEnumerator1
    Sprite Size Mode=::NewEnumerator1
    Uniform Sprite Size Max=35
    Uniform Sprite Size Min=22
    UsePositionOffset=false
    Write Material Random=true
   )
   &ShapeLocation(MaskSpawnGroup=false, Non Uniform Scale=<1.7 3.5 0.3>, Override Seed=false, Sphere Radius=32, UseOwnerScale=true)
   &AddVelocity(Velocity=&RandomRangeVector(Minimum=<0 0 5>, Maximum=<0 0 25>)) {module=AddVelocity001}
   &AddVelocity(
    Cone Angle=150
    Cone Axis=<0 0 1>
    Module.UseConeFalloff=false
    Module.UseVelDistribution=false
    Override Seed=false
    Velocity Mode=::NewEnumerator2
    Velocity Speed=&RandomRangeFloat(Minimum=30, Maximum=170)
   )
   &ApplyInitialForces()
   &CameraOffset(Camera Offset Amount=&RandomRangeFloat(Minimum=0, Maximum=15))
  ]
  particleUpdateScript [
   &ParticleState()
   &ScaleColor(
    Scale Alpha=&FloatFromCurve(FloatCurve=DataInterface() {propertyValues {Curve {keys [
         <at 0 0 RCIM_Cubic>
         {arriveTangent=1.4793, interpMode=RCIM_Cubic, leaveTangent=1.4793, time=0.1522, value=0.8624}
         {arriveTangent=-1.6626, interpMode=RCIM_Cubic, leaveTangent=-1.6626, tangentMode=RCTM_User, time=0.3535, value=0.523}
         <at 1 0 RCIM_Cubic>
        ]}}})
    ScaleA=true
    ScaleRGB=true
   )
   &Drag(Drag=&RandomRangeFloat(Minimum=3, Maximum=6), Use Linear Drag=true, Use Rotational Drag=true)
   &SolveForcesAndVelocity(Clamp Velocity=false, Limit Acceleration=false)
   &DynamicMaterialParameters(
    Index 0 Param 1=&FloatFromCurve(FloatCurve=DataInterface() {propertyValues {Curve {keys [<at 0 0 RCIM_Cubic>,<at 1.0011 0.2>]}}})
    Index 0 Param 2=&FloatFromCurve(FloatCurve=DataInterface() {propertyValues {Curve {keys [
         {arriveTangent=0.0453, interpMode=RCIM_Cubic, tangentMode=RCTM_Break, time=0, value=-1}
         {arriveTangent=0.0604, interpMode=RCIM_Cubic, leaveTangent=0.0604, tangentMode=RCTM_User, time=1, value=-1}
        ]}}})
    Index 0 Param 3=&FloatFromCurve(FloatCurve=DataInterface() {propertyValues {Curve {keys [{interpMode=RCIM_Cubic, leaveTangent=4.6779, time=0.2851, value=0}]}}})
    Index 1 Param 1=&FloatFromCurve(
     FloatCurve=DataInterface() {propertyValues {Curve {keys [<at 0 0>,<at 0.25 1>]}}}
     Scale Curve=&RandomRangeFloat(Minimum=2, Maximum=1)
    )
    Index 1 Param 2=&FloatFromCurve(
     FloatCurve=DataInterface() {propertyValues {Curve {keys [
         {interpMode=RCIM_Cubic, leaveTangent=3.7422, tangentMode=RCTM_Break, time=0, value=0.2}
         {interpMode=RCIM_Cubic, leaveTangent=0.9066, time=1, value=2}
        ]}}}
     Scale Curve=&RandomRangeFloat(Minimum=0.5, Maximum=1)
    )
    Index 1 Param 3=&FloatFromCurve(FloatCurve=DataInterface() {propertyValues {Curve {keys [
         {interpMode=RCIM_Cubic, leaveTangent=2.8445, tangentMode=RCTM_Break, time=0, value=0.2}
         {interpMode=RCIM_Cubic, leaveTangent=63.2978, time=0.25, value=1}
        ]}}})
    Index 1 Param 4=&FloatFromCurve(
     FloatCurve=DataInterface() {propertyValues {Curve {keys [<at 0 1>,<at 1 0.8>]}}}
     Scale Curve=&RandomRangeFloat(Minimum=0.5, Maximum=1)
    )
    Index 2 Param 1=&FloatFromCurve(FloatCurve=DataInterface() {propertyValues {Curve {keys [{interpMode=RCIM_Cubic, leaveTangent=3.2309, tangentMode=RCTM_Break, time=0.1, value=0}, <at 1 1>]}}})
    Index 2 Param 2=$Particles.MaterialRandom
    Index 2 Param 3=&FloatFromCurve(FloatCurve=DataInterface() {propertyValues {Curve {keys [
         {interpMode=RCIM_Cubic, tangentMode=RCTM_Break, time=0, value=1}
         {interpMode=RCIM_Cubic, leaveTangent=11.1233, tangentMode=RCTM_Break, time=1, value=2}
        ]}}})
    Index 2 Param 4=0.75
    Module.Param0WriteEnabled=true
    Module.Param0WriteEnabled1=true
    Module.Param0WriteEnabled2=true
    Module.Param1WriteEnabled=false
    Module.Param1WriteEnabled1=true
    Module.Param1WriteEnabled2=true
    Module.Param2WriteEnabled=true
    Module.Param2WriteEnabled1=true
    Module.Param2WriteEnabled2=true
    Module.Param3WriteEnabled=true
    Module.Param3WriteEnabled1=true
    Module.Param3WriteEnabled2=true
    Write Parameter Index 1=true
    Write Parameter Index 2=true
   )
   /SetVariables(Particles.SpriteSize=&Multiply_Vector2D(
     A=$Particles.Initial.SpriteSize
     B=&Vector2DFromFloat(Value=&FloatFromCurve(
       FloatCurve=DataInterface() {propertyValues {Curve {keys [{tangentMode=RCTM_User, time=0, value=0}, <at 0.04 1.27>, <at 0.14 1.6>, {tangentMode=RCTM_User, time=1, value=3.45}]}}}
       Scale Curve=&RandomRangeFloat(Minimum=1, Maximum=1.3)
      ))
    )) {isSetParameters=true}
  ]
  rendererClasses [&NiagaraSpriteRendererProperties]
  renderers [{rendererClass=&NiagaraSpriteRendererProperties}]
 }
 LeafBurst {
  emitterUpdateScript [
   &EmitterState(Life Cycle Mode=::NewEnumerator1, Loop Behavior=::NewEnumerator1, Loop Delay=0.23, Loop Duration=1, UseLoopDelay=true)
   &SpawnBurst_Instantaneous(Spawn Count=25, Use Spawn Probability=false, UseLoopCountLimit=false)
  ]
  particleSpawnScript [
   &InitializeParticle(
    Color Maximum=<rgba 0.1422 0.2865 0.0863 1>
    Color Minimum=<rgba 0.2392 0.3646 0.1487 1>
    Color Mode=::NewEnumerator2
    Lifetime Max=3.25
    Lifetime Mode=::NewEnumerator1
    Mesh Scale Mode=::NewEnumerator3
    Mesh Uniform Scale=&UniformAOrBFloat(
     A=&RandomRangeFloat(Minimum=0.05, Maximum=0.3)
     B=&RandomRangeFloat(Minimum=0.33, Maximum=0.4)
     Distribution Weight=0.2
     Evaluation Type=::NewEnumerator0
     Recalculate Random Each Loop=true
     Override Seed=false
    )
    UsePositionOffset=false
    Write Material Random=true
   )
   &ShapeLocation(
    MaskSpawnGroup=false
    Non Uniform Scale=<1.25 1.5 1>
    Offset=<0 -5 12>
    Offset Mode=::NewEnumerator0
    Override Seed=false
    Sphere Radius=50
    UseOwnerScale=true
   )
   &AddVelocity(
    Cone Angle=180
    Cone Axis=<0 0 1>
    Module.UseConeFalloff=false
    Module.UseVelDistribution=false
    Override Seed=false
    Velocity Mode=::NewEnumerator2
    Velocity Speed=&RandomRangeFloat(Minimum=700, Maximum=500)
   )
   &AddVelocity(Velocity=&RandomRangeVector(Minimum=<0 0 70>, Maximum=<0 0 125>)) {module=AddVelocity001}
   &InitialMeshOrientation(
    Mesh Orientation Mode=::NewEnumerator2
    Orientation Vector=&Lerp_Vector(
     A=&RandomRangeVector(Minimum=<0 1 0>, Maximum=<1 1 1>)
     B=&RandomRangeVector(Minimum=<0 0 0>, Maximum=<1 1 1>)
     Alpha=&UniformAOrBFloat(A=0, Evaluation Type=::NewEnumerator0, Recalculate Random Each Loop=true, Override Seed=false)
    )
    Use Rotation Vector=false
   )
  ]
  particleUpdateScript [
   &ParticleState()
   &UpdateMeshOrientation(Rotation Vector=&RandomRangeVector(Minimum=<0 0 0>, Maximum=<1 1 1>))
   &ScaleMeshSize(Scale Factor=&VectorFromFloat(Value=&FloatFromCurve(FloatCurve=DataInterface() {propertyValues {Curve {
         defaultValue=1
         keys [
          {interpMode=RCIM_Cubic, tangentMode=RCTM_User, time=0, value=0}
          {arriveTangentWeight=0.0233, interpMode=RCIM_Cubic, leaveTangentWeight=0.2102, tangentMode=RCTM_User, time=0.07, value=1}
          {arriveTangent=-1.0386, interpMode=RCIM_Cubic, leaveTangent=-1.0386, tangentMode=RCTM_User, time=0.1431, value=0.9958}
          {arriveTangent=-2.0738, interpMode=RCIM_Cubic, leaveTangent=-2.0738, tangentMode=RCTM_User, time=0.2622, value=0.2638}
          <at 0.6 0 RCIM_Cubic>
         ]
        }}})))
   &GravityForce()
   &ScaleColor(
    Scale Alpha=&FloatFromCurve(FloatCurve=DataInterface() {propertyValues {Curve {keys [<at -0.0046 1.0113 RCIM_Cubic>,<at 1.0011 0.002 RCIM_Cubic>]}}})
    ScaleA=true
    ScaleRGB=true
   )
   &Drag(
    Drag=&FloatFromCurve(
     FloatCurve=DataInterface() {propertyValues {Curve {keys [<at 0.1 1>,<at 0.5 2>]}}}
     Scale Curve=&RandomRangeFloat(Minimum=8, Maximum=11)
    )
    Use Linear Drag=true
    Use Rotational Drag=true
   )
   &SolveForcesAndVelocity(Clamp Velocity=false, Limit Acceleration=false)
  ]
  rendererClasses [&NiagaraMeshRendererProperties]
  renderers [{rendererClass=&NiagaraMeshRendererProperties, rendererIndex=0}]
 }
 UmbrellaRocks {
  emitterUpdateScript [
   &EmitterState(Life Cycle Mode=::NewEnumerator1, Loop Behavior=::NewEnumerator1, Loop Delay=0.1, Loop Duration=4, UseLoopDelay=true)
   &SpawnBurst_Instantaneous(
    Spawn Count=&RandomRangeInt(Minimum=9, Maximum=15, Recalculate Random Each Loop=false)
    Use Spawn Probability=false
    UseLoopCountLimit=false
   )
   &SpawnBurst_Instantaneous(Spawn Count=5, Spawn Group=1, Use Spawn Probability=false, UseLoopCountLimit=false) {module=SpawnBurst_Instantaneous001}
   &SpawnBurst_Instantaneous(Spawn Count=30, Spawn Group=2, Use Spawn Probability=false, UseLoopCountLimit=false) {module=SpawnBurst_Instantaneous002}
  ]
  particleSpawnScript [
   /SetVariables(Particles.VisibilityTag=&SpawnGroup()) {isSetParameters=true, module=SetVariables_#2}
   &InitializeParticle(
    Color=<rgba 0.1406 0.0699 0.0342 1>
    Lifetime=&MergeHelper_FloatSwitch(
     0=&RandomRangeFloat(Minimum=0.9, Maximum=1)
     1=&RandomRangeFloat(Minimum=0.9, Maximum=1)
     2=&RandomRangeFloat(Minimum=0.6, Maximum=1)
     3=?
     Index=&SpawnGroup()
     Num=::NewEnumerator1
    )
    Mass Max=1.5
    Mass Mode=::NewEnumerator3
    Mesh Scale Mode=::NewEnumerator3
    Mesh Uniform Scale=&MaskFloatBySpawnGroup(
     Masked Float=&RandomRangeFloat(Minimum=0.05, Maximum=0.1)
     Passthrough Float=&RandomRangeFloat(Minimum=0.15, Maximum=0.19)
    )
    Sprite Rotation Mode=::NewEnumerator2
    Sprite Size Mode=::NewEnumerator3
    Uniform Sprite Size=&RandomRangeFloat(Minimum=3, Maximum=5)
    UsePositionOffset=true
    Write Material Random=true
   )
   &ShapeLocation(
    MaskSpawnGroup=false
    Non Uniform Scale=&MaskVector(Mask Source=1, Mask Parameter=$Particles.PartitionParticles.Partition, Mask=0, Masked Vector=<1 1 1>, Passthrough Vector=<1 3 1>)
    Offset=&MaskVector(Mask Source=1, Mask Parameter=$Particles.PartitionParticles.Partition, Mask=0, Masked Vector=<45 -60 12>, Passthrough Vector=<0 0 12>)
    Offset Mode=::NewEnumerator0
    Override Seed=false
    Ring Radius=&MaskFloat(Mask Source=1, Mask Parameter=$Particles.PartitionParticles.Partition, Mask=0, Masked Float=5, Passthrough Float=23)
    Shape Primitive=::NewEnumerator4
    UseOwnerScale=true
   )
   &InitialMeshOrientation(Use Rotation Vector=false)
   &AddVelocity(
    Cone Angle=&MaskFloatBySpawnGroup(
     Masked Float=&RandomRangeFloat(Minimum=40.8464, Maximum=186.7016)
     Passthrough Float=&RandomRangeFloat(Minimum=14.1396, Maximum=44.0656)
    )
    Module.UseConeFalloff=false
    Module.UseVelDistribution=false
    Override Seed=false
    Rotation Angle=180
    Rotation Mode=::NewEnumerator1
    Velocity Mode=::NewEnumerator2
    Velocity Speed=&MergeHelper_FloatSwitch(
     0=&RandomRangeFloat(Minimum=50, Maximum=350)
     1=&RandomRangeFloat(Minimum=20, Maximum=200)
     2=&RandomRangeFloat(Minimum=100, Maximum=400)
     3=?
     Index=&SpawnGroup()
     Num=::NewEnumerator1
    )
   )
  ]
  particleUpdateScript [
   &GravityForce(Gravity=<0 0 -1180.0895>)
   &Collision(
    Age Colliding Particles=false
    Collision Enabled=&Boolean_OrOperation(
     Bool1=&MaskBool(
      Mask Source:NiagaraInt32=0
      Mask Parameter:NiagaraInt32=?
      Mask:NiagaraInt32=2
      Masked Bool:NiagaraBool=false
      Passthrough Bool:NiagaraBool=&Boolean_OrOperation(
       Bool1=&IsPlatformSetActive(New Platform Set=DataInterface() {propertyValues {Platforms {qualityLevelMask=-8}}})
       Bool2=$User.bIsFrontEndPreview
      )
     )
     Bool2=false
    )
    Friction=0.5
    Friction During a Bounce=0.3
    KillOnCollision=false
    Module.Control Rest State Via Surface Normal=false
    Module.EnableMaxCollisionCount=false
    Module.Randomize Collision Normal Vector=false
    Radius Calculation Type=::NewEnumerator1
    Restitution=0.44
   )
   &Collision(
    Age Colliding Particles=false
    Analytical Collision Plane Position 1:Vector3f=<0 0 0>
    Analytical Collision Plane Position 2:Vector3f=<0 0 0>
    Analytical Collision Plane Space:ENiagaraCoordinateSpace=::Local
    CPU Collision Type=::NewEnumerator3
    Collision Enabled=&Boolean_OrOperation(
     Bool1=&MaskBool(
      Mask Source:NiagaraInt32=0
      Mask Parameter:NiagaraInt32=?
      Mask:NiagaraInt32=2
      Masked Bool:NiagaraBool=false
      Passthrough Bool:NiagaraBool=&Boolean_OrOperation(
       Bool1=&IsPlatformSetActive(New Platform Set=DataInterface() {propertyValues {Platforms {qualityLevelMask=-28}}})
       Bool2=$User.bIsFrontEndPreview
      )
     )
     Bool2=false
    )
    Friction=0.5
    Friction During a Bounce=0.3
    KillOnCollision=false
    Module.Control Rest State Via Surface Normal=false
    Module.EnableMaxCollisionCount=false
    Module.Randomize Collision Normal Vector=false
    Radius Calculation Type=::NewEnumerator1
    Restitution=0.44
   ) {module=Collision001}
   &SolveForcesAndVelocity(Clamp Velocity=false, Limit Acceleration=false)
   &ParticleState()
   &ScaleMeshSize(Scale Factor=&VectorFromCurve(VectorCurve=DataInterface() {propertyValues {
       XCurve {keys [<at 0 0.35 RCIM_Cubic>, {arriveTangent=-0.35, interpMode=RCIM_Cubic, leaveTangent=-0.35, time=0.1, value=1}, <at 1 0 RCIM_Cubic>]}
       YCurve {keys [<at 0 0.35 RCIM_Cubic>, {arriveTangent=-0.35, interpMode=RCIM_Cubic, leaveTangent=-0.35, time=0.1, value=1}, <at 1 0 RCIM_Cubic>]}
       ZCurve {keys [<at 0 0.35 RCIM_Cubic>, {arriveTangent=-0.35, interpMode=RCIM_Cubic, leaveTangent=-0.35, time=0.1, value=1}, <at 1 0 RCIM_Cubic>]}
      }}))
   &DynamicMaterialParameters(Module.Index0Float0Write=true, Module.Index0Float1Write=true, Module.Index0Float2Write=true, Module.Index0Float3Write=true)
   /SetVariables(Particles.float:NiagaraFloat=&FloatFromCurve(
     FloatCurve=DataInterface() {propertyValues {Curve {
        defaultValue=1
        keys [
         {arriveTangent=4.1785, interpMode=RCIM_Cubic, leaveTangent=4.1785, tangentMode=RCTM_User, time=0, value=0}
         <at 0.2318 0.8879 RCIM_Cubic>
         {arriveTangent=0.0224, interpMode=RCIM_Cubic, leaveTangent=0.0224, tangentMode=RCTM_User, time=0.9929, value=0}
        ]
       }}}
     CurveIndex=$Particles.Age
    )) {isSetParameters=true, module=SetVariables_#3}
  ]
  rendererClasses [&NiagaraMeshRendererProperties, &NiagaraSpriteRendererProperties]
  renderers [
   {rendererClass=&NiagaraMeshRendererProperties, rendererIndex=0}
   {rendererClass=&NiagaraSpriteRendererProperties, rendererIndex=1}
   {rendererClass=&NiagaraMeshRendererProperties, rendererIndex=2}
   {rendererClass=&NiagaraSpriteRendererProperties, rendererIndex=3}
  ]
 }
 UmbrellaSmoke {rendererClasses [], renderers []}
}
summary {
 systemName=NS_Emote_CrimpPerfume_PropSpawn
 emitters [
  {emitterName=UmbrellaRocks, rendererClasses [&NiagaraMeshRendererProperties, &NiagaraSpriteRendererProperties]}
  {emitterName=LeafBurst, rendererClasses [&NiagaraMeshRendererProperties]}
  {emitterName=UmbrellaSmoke, rendererClasses []}
  {emitterName=ChairRocks, rendererClasses [&NiagaraMeshRendererProperties, &NiagaraSpriteRendererProperties]}
  {emitterName=Chair_Smoke, rendererClasses [&NiagaraSpriteRendererProperties]}
 ]
 userVariables [
  User.bIsFrontEnd {type=NiagaraBool}
  User.bIsFrontEndPreview {type=NiagaraBool}
 ]
}
